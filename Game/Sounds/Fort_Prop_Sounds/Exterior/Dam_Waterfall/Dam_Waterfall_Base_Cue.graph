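# Sound cue for the looping close-range roar at the base of the dam waterfall, routed to the water interaction sound class.
# A Random node picks one of three equally weighted looping wave variants.
# Overridden attenuation gives it a wide audible radius with stereo spread, distance low-pass filtering, listener focus and interpolated occlusion on a game trace channel.
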SoundCue /Game/Sounds/Fort_Prop_Sounds/Exterior/Dam_Waterfall/Dam_Waterfall_Base_Cue
root=Random_0
nodes [
 {id=Random_0, children [WavePlayer_3,WavePlayer_4,WavePlayer_5], properties {weights [1,1,1]}}
 {
  id=WavePlayer_3
  properties {bLooping=true, soundWaveAssetPtr=&/Dam_Waterfall_Base_Close_01a}
 }
 {
  id=WavePlayer_4
  properties {bLooping=true, soundWaveAssetPtr=&/Dam_Waterfall_Base_Close_01b}
 }
 {
  id=WavePlayer_5
  properties {bLooping=true, soundWaveAssetPtr=&/Dam_Waterfall_Base_Close_01c}
 }
]
properties {
 bOverrideAttenuation=true
 maxDistance=9500
 soundClassObject=/Game/Sounds/FortSoundClassesAndModes/Fort_Water_Interaction
 volumeMultiplier=1
 attenuationOverrides {
  bAttenuateWithLPF=true
  bEnableFocusInterpolation=true
  bEnableListenerFocus=true
  bEnableOcclusion=true
  falloffDistance=7500
  lPFFrequencyAtMax=1000
  lPFRadiusMax=4000
  lPFRadiusMin=1000
  nonFocusVolumeAttenuation=0.8
  nonSpatializedRadiusStart=3000
  occlusionInterpolationTime=2
  occlusionLowPassFilterFrequency=1000
  occlusionTraceChannel=ECC_GameTraceChannel11
  stereoSpread=500
  attenuationShapeExtents {x=2000}
 }
 concurrencyOverrides {maxCount=2}
}
