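# Looping ambient sound cue for the dam waterfall intake, routed through the Fort_Ambient_TileMix sound class.
# A single wave player loops Dam_Waterfall_Intake_01 as the root, with no randomization or mixing.
# An attenuation override fades the sound by distance from a capsule shape with spatialization turned off, so it is not panned by position.
# Occlusion traces on a custom game channel apply an interpolated low-pass filter when the source is blocked.
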SoundCue /Game/Sounds/Fort_Prop_Sounds/Exterior/Dam_Waterfall/Dam_Waterfall_Intake_Cue
root=WavePlayer_0
nodes [{id=WavePlayer_0, properties {bLooping=true, soundWaveAssetPtr=&/Dam_Waterfall_Intake_01}}]
properties {
 bOverrideAttenuation=true
 maxDistance=3181.8208
 soundClassObject=/Game/Sounds/FortSoundClassesAndModes/Fort_Ambient_TileMix
 volumeMultiplier=1
 attenuationOverrides {
  attenuationShape=Capsule
  bEnableOcclusion=true
  bSpatialize=false
  falloffDistance=2353.8157
  occlusionInterpolationTime=1
  occlusionLowPassFilterFrequency=1000
  occlusionTraceChannel=ECC_GameTraceChannel11
  attenuationShapeExtents=<828.0052 167.2599>
 }
}
